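# Physics asset for the BitterSweet backpack cosmetic, defining secondary-motion bodies for its dangling bullets, grenades and umbrella.
# Base bodies attach to MSTt_middlePoint, mostly angularly locked, and anchor Dyn bodies with narrow swing and twist limits.
# A four-link bullet chain and a grenade-to-ring chain add serial movement, while Base_grenade_B moves directly off the middle point.
# Shapes are capsules plus one box on the umbrella; Spine_joint has no constraint.
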PhysicsAsset /BRCosmetics/Accessories/FORT_Backpacks/Backpack_M_MED_BitterSweet/Meshes/M_MED_BitterSweet_Pack_Physics
bodies {
 Base_bullet [{shapeType=Capsule, length=0.505, radius=0.505, shapeName=Base_bullet_capsule, rotation=<rot 0 -90 0>}]
 Base_bullet_A [{shapeType=Capsule, length=0.505, radius=0.505, shapeName=Base_bullet_A_capsule, rotation=<rot 0 -90 0>}]
 Base_bullet_B [{shapeType=Capsule, length=0.505, radius=0.505, shapeName=Base_bullet_B_capsule, rotation=<rot 0 -90 0>}]
 Base_bullet_C [{shapeType=Capsule, length=0.505, radius=0.505, shapeName=Base_bullet_C_capsule, rotation=<rot 0 -90 0>}]
 Base_grenade_A [{shapeType=Capsule, length=0.505, radius=0.505, shapeName=Base_grenade_A_capsule, rotation=<rot 0 -90 0>}]
 Base_grenade_B [{shapeType=Capsule, length=3.8935, radius=3, shapeName=Base_grenade_B_capsule, rotation=<rot 0 -86.4986 -85.0069>}]
 Base_grenade_C [{shapeType=Capsule, length=0.505, radius=0.505, shapeName=Base_grenade_C_capsule}]
 Base_grenade_Ring [{shapeType=Capsule, length=0.505, radius=0.505, shapeName=Base_grenade_Ring_capsule, rotation=<rot 0 -90 0>}]
 Base_umbrella [{shapeType=Capsule, length=0.505, radius=0.505, shapeName=Base_umbrella_capsule, rotation=<rot 0 -90 0>}]
 Dyn_bullet_A [{shapeType=Capsule, length=1.5108, radius=0.4795, shapeName=Dyn_bullet_A_capsule}]
 Dyn_bullet_B [{shapeType=Capsule, length=2.3765, radius=0.4844, shapeName=Dyn_bullet_B_capsule}]
 Dyn_bullet_C [{shapeType=Capsule, length=2.0122, radius=0.458, shapeName=Dyn_bullet_C_capsule}]
 Dyn_grenade_A [{
   shapeType=Capsule
   length=2.2257
   radius=3.2038
   shapeName=Dyn_grenade_A_capsule
   center=<-3.2878 0.4088 0.4563>
   rotation=<rot 79.7898 0.0002 -12.807>
  }]
 Dyn_grenade_C [{shapeType=Capsule, length=7.2767, radius=3.6415, shapeName=Dyn_grenade_C_capsule, center=<-0.0987 0 -0.8545>, rotation=<rot 83.4127 0 0>}]
 Dyn_grenade_Ring [{shapeType=Capsule, length=1.5349, radius=1.5287, shapeName=Dyn_grenade_Ring_capsule, center=<1.9132 -0.8327 0>}]
 Dyn_umbrella [
  {
   shapeType=Capsule
   length=9.1172
   radius=0.6453
   shapeName=Dyn_umbrella_capsule
   center=<-7.4867 0 -0.4641>
   rotation=<rot 86.4525 179.9995 179.9994>
  }
  {
   shapeType=Box
   extentX=1.4857
   extentY=13.4729
   extentZ=12.8368
   shapeName=Dyn_umbrella_box
   center=<-15.0821 0 0>
   rotation=<rot 0 0.0002 6.6097>
  }
 ]
 MSTt_middlePoint [{shapeType=Capsule, length=4.1664, radius=8.3166, center=<-3.3619 -2.0604 -1.6138>, rotation=<rot -52.333 -128.7849 -146.9874>}]
 Spine_joint [{shapeType=Capsule, length=4, radius=10.5, shapeName=Spine_joint_capsule, rotation=<rot 0 0 -90>}]
 dyn_bullet_01 [{shapeType=Capsule, length=2.3575, radius=2.1428, shapeName=dyn_bullet_01_capsule}]
 dyn_bullet_02 [{shapeType=Capsule, length=2.1942, radius=2.0789, shapeName=dyn_bullet_02_capsule, center=<0.4729 0 0>}]
 dyn_bullet_03 [{shapeType=Capsule, length=2.3638, radius=2.2819, shapeName=dyn_bullet_03_capsule, center=<0.5678 0 0>}]
 dyn_bullet_04 [{shapeType=Capsule, length=3.622, radius=3.3495, shapeName=dyn_bullet_04_capsule, center=<2.0979 0 0>, rotation=<rot 90 166.7373 166.7372>}]
}
constraints [
 {
  bone1Name=Base_bullet
  bone2Name=MSTt_middlePoint
  swing1LimitDegrees=30
  swing1Motion=Locked
  swing2LimitDegrees=5
  swing2Motion=Locked
  twistLimitDegrees=5
  twistMotion=Locked
 }
 {bone1Name=dyn_bullet_01, bone2Name=Base_bullet, swing1LimitDegrees=30, swing2LimitDegrees=10, twistLimitDegrees=0}
 {bone1Name=dyn_bullet_02, bone2Name=dyn_bullet_01, swing1LimitDegrees=30, swing2LimitDegrees=10, twistLimitDegrees=0}
 {bone1Name=dyn_bullet_03, bone2Name=dyn_bullet_02, swing1LimitDegrees=30, swing2LimitDegrees=10, twistLimitDegrees=0}
 {bone1Name=dyn_bullet_04, bone2Name=dyn_bullet_03, swing1LimitDegrees=30, swing2LimitDegrees=5, twistLimitDegrees=0}
 {
  bone1Name=Base_grenade_A
  bone2Name=MSTt_middlePoint
  swing1LimitDegrees=7
  swing1Motion=Locked
  swing2LimitDegrees=3
  swing2Motion=Locked
  twistLimitDegrees=5
  twistMotion=Locked
 }
 {bone1Name=Dyn_grenade_A, bone2Name=Base_grenade_A, swing1LimitDegrees=7, swing2LimitDegrees=1, twistLimitDegrees=1}
 {bone1Name=Base_grenade_Ring, bone2Name=Dyn_grenade_A, swing1LimitDegrees=17, swing2LimitDegrees=2, twistLimitDegrees=2}
 {bone1Name=Dyn_grenade_Ring, bone2Name=Base_grenade_Ring, swing1LimitDegrees=17, swing2LimitDegrees=2, twistLimitDegrees=2}
 {bone1Name=Base_grenade_B, bone2Name=MSTt_middlePoint, swing1LimitDegrees=7, swing2LimitDegrees=3, twistLimitDegrees=1}
 {
  bone1Name=Base_grenade_C
  bone2Name=MSTt_middlePoint
  swing1LimitDegrees=1
  swing1Motion=Locked
  swing2LimitDegrees=1
  swing2Motion=Locked
  twistLimitDegrees=1
  twistMotion=Locked
 }
 {bone1Name=Dyn_grenade_C, bone2Name=Base_grenade_C, swing1LimitDegrees=20, swing2LimitDegrees=5, twistLimitDegrees=3}
 {bone1Name=Dyn_umbrella, bone2Name=Base_umbrella, swing1LimitDegrees=8, swing2LimitDegrees=14, twistLimitDegrees=0, twistMotion=Locked}
 {
  bone1Name=Base_umbrella
  bone2Name=MSTt_middlePoint
  swing1LimitDegrees=20
  swing1Motion=Locked
  swing2LimitDegrees=2
  swing2Motion=Locked
  twistLimitDegrees=2
  twistMotion=Locked
 }
 {bone1Name=Base_bullet_A, bone2Name=MSTt_middlePoint, swing1LimitDegrees=11, swing2LimitDegrees=11, twistLimitDegrees=5}
 {bone1Name=Dyn_bullet_A, bone2Name=Base_bullet_A, swing1LimitDegrees=7, swing2LimitDegrees=7, twistLimitDegrees=2}
 {
  bone1Name=Base_bullet_B
  bone2Name=MSTt_middlePoint
  swing1LimitDegrees=3
  swing1Motion=Locked
  swing2LimitDegrees=3
  swing2Motion=Locked
  twistLimitDegrees=3
  twistMotion=Locked
 }
 {bone1Name=Dyn_bullet_B, bone2Name=Base_bullet_B, swing1LimitDegrees=7, swing2LimitDegrees=7, twistLimitDegrees=7}
 {
  bone1Name=Base_bullet_C
  bone2Name=MSTt_middlePoint
  swing1LimitDegrees=3
  swing1Motion=Locked
  swing2LimitDegrees=3
  swing2Motion=Locked
  twistLimitDegrees=3
  twistMotion=Locked
 }
 {bone1Name=Dyn_bullet_C, bone2Name=Base_bullet_C, swing1LimitDegrees=7, swing2LimitDegrees=7, twistLimitDegrees=7}
]
